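# Player anim layer Blueprint for the DragonCart SMG that adds a one-shot reaction when upper-body targeting starts.
# The update event detects the rising edge of bPlayUpperBodyTargeting and sets Targeting_JustActivated.
# In MovementMode_SourcePose, that flag sends a state machine into TargetedReaction, its entry state, which applies an additive sequence over the cached source pose, then automatically returns to plain source pose.
# The other MovementMode layers contain no logic.
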AnimBlueprint /DragonCartWeaponGameplay/Gameplay/DragonCart_SMG/Animation/DragonCart_SMG_Player_LayerAnimBP
target_skeleton=&Fortnite_M_Avg_Player_Skeleton
graphs [
 AnimGraph {nodes [Output Pose() {layerGroup=None}]}
 EventGraph {nodes [
   Event BlueprintUpdateAnimation()
   n1 = Get Targeting_Previous() -> (Targeting_Previous:bool)
   n2 = Get bPlayUpperBodyTargeting() -> (bPlayUpperBodyTargeting:bool)
   > n3 = Branch(Condition=AND Boolean(A="Not Equal (Boolean)"(A=@n1, B=@n2) -> (ReturnValue=~), B=@n2))
   > n4 = Set Targeting_JustActivated(Targeting_JustActivated:bool=true) -> (Output_Get:bool)
   Set Targeting_Previous(execute=[@n5.then, @n4.then], Targeting_Previous:bool=Get bPlayUpperBodyTargeting()) -> (Output_Get:bool)
   n5 = Set Targeting_JustActivated(execute=@n3.else, Targeting_JustActivated=false)
   TryGetPawnOwner()
  ]}
 MovementMode_AimOffset {nodes [
   Input Pose() {name=InAimOffsetsPose}
   Input Pose() {name=InSourcePosePreAimOffsets}
   Output Pose() {layerGroup=MovementMode}
  ]}
 MovementMode_BaseLocomotion {nodes [
   Input Pose() {name=InPoseBaseLocomotion}
   Output Pose() {layerGroup=MovementMode}
  ]}
 MovementMode_FinalPoseOverride {nodes [
   Input Pose() {name=InFinalPose}
   Output Pose() {layerGroup=MovementMode}
  ]}
 MovementMode_FullBodyOverride {nodes [
   Input Pose() {name=InputPoseFullBody}
   Output Pose() {layerGroup=MovementMode}
  ]}
 MovementMode_InAirOverride {nodes [
   Input Pose() {name=InSource_Pose}
   Input Pose() {name=InAirPose}
   Output Pose() {layerGroup=MovementMode}
  ]}
 MovementMode_LocomotionAdditives {nodes [
   Input Pose() {name=InLocomotionAdditivesPose}
   Output Pose() {layerGroup=MovementMode}
  ]}
 MovementMode_LowerBodyOverride {nodes [
   Input Pose() {name=InLowerBodyPose}
   Input Pose() {name=InSourcePose}
   Output Pose() {layerGroup=MovementMode}
  ]}
 MovementMode_OffScreen {nodes [
   Input Pose() {name=InPoseOffscreen}
   Output Pose() {layerGroup=MovementMode}
  ]}
 MovementMode_SourcePose {nodes [
   AnimGraphNode_StateMachine_4 = TargetedState() {is=title AnimGraphNode_StateMachine, bAllowConduitEntryStates=true, bReinitializeOnBecomingRelevant=false}
   Output Pose(Result=@AnimGraphNode_StateMachine_4) {layerGroup=MovementMode}
   Save cached pose 'InSourcePose'(Pose=Input Pose() {name=InSourcePoseBase}) {cacheName=InSourcePose}
  ]}
 TargetedState {
  entry=TargetedReaction
  path=MovementMode_SourcePose.AnimGraphNode_StateMachine_4.TargetedState
  ids=id
  nodes [
   AnimStateNode_1 = TargetedReaction() {is=title AnimStateNode, bAlwaysResetOnEntry=true, stateLeft {guid=#1, notifyName=Targeted_Out}}
   AnimStateNode_3 = TargetedSourcePose() {is=title AnimStateNode}
   AnimStateTransitionNode_3 = AnimStateTransitionNode() {bAutomaticRuleBasedOnSequencePlayerInState=true}
   AnimStateTransitionNode_4 = AnimStateTransitionNode() {bAllowInertializationForSelfTransitions=true}
  ]
  links [
   {
    dst=TargetedSourcePose
    rule=MovementMode_SourcePose.AnimGraphNode_StateMachine_4.TargetedState.AnimStateTransitionNode_3.Transition
    src=TargetedReaction
    via=AnimStateTransitionNode_3
   }
   {
    dst=TargetedReaction
    rule=MovementMode_SourcePose.AnimGraphNode_StateMachine_4.TargetedState.AnimStateTransitionNode_4.Transition
    src=TargetedSourcePose
    via=AnimStateTransitionNode_4
   }
  ]
 }
 TargetedReaction {
  path=MovementMode_SourcePose.AnimGraphNode_StateMachine_4.TargetedState.AnimStateNode_1.TargetedReaction
  nodes [
   n1 = "Sequence Player (additive) 'DragonCart_SMG_NonTargetToTarget_CMM'"() {bLoopAnimation=false, sequence=/DragonCartWeaponGameplay/Gameplay/DragonCart_SMG/Animation/Poses/DragonCart_SMG_NonTargetToTarget_CMM}
   Output Animation Pose(Result=Apply Additive(Base=Use cached pose 'InSourcePose'(), Additive=@n1) {alphaBoolBlend {blendOption=Linear}})
  ]
 }
 TargetedSourcePose {
  path=MovementMode_SourcePose.AnimGraphNode_StateMachine_4.TargetedState.AnimStateNode_3.TargetedSourcePose
  nodes [Output Animation Pose(Result=Use cached pose 'InSourcePose'())]
 }
 Transition {path=MovementMode_SourcePose.AnimGraphNode_StateMachine_4.TargetedState.AnimStateTransitionNode_3.Transition, nodes [Result()]}
 Transition {
  path=MovementMode_SourcePose.AnimGraphNode_StateMachine_4.TargetedState.AnimStateTransitionNode_4.Transition
  nodes [Result(bCanEnterTransition=Get Targeting_JustActivated() -> (Targeting_JustActivated:bool))]
 }
]
variables [
 Targeting_JustActivated {type=boolean}
 Targeting_Previous {type=boolean}
]
