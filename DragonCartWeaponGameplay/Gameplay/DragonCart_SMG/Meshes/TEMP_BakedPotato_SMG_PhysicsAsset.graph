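# Physics asset for the temporary BakedPotato SMG skeletal mesh in the DragonCart SMG weapon folder.
# It holds a single body on the handle bone, made of one rotated capsule as a simple collision approximation.
# It defines no constraints, so there are no jointed or separately simulated parts.
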PhysicsAsset /DragonCartWeaponGameplay/Gameplay/DragonCart_SMG/Meshes/TEMP_BakedPotato_SMG_PhysicsAsset
bodies {handle [{shapeType=Capsule, length=31.5768, radius=11.6211, center=<5.6405 0.1889 4.4145>, rotation=<rot 0 90.1829 -84.437>}]}
constraints []
